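; Parent looping gameplay cue that guides the target player to the nearest locked keycard terminal.
; On application it caches pawn and controller, then starts a repeating timer whose interval comes from the cue's RawMagnitude.
; Each tick re-marks the terminal returned by a child-overridden search if it changed, and spawns a world-rotated direction emitter on the pawn.
; An AbilityLevel below 1 or cue removal stops the timer; removal leaves the marker.
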(event EventOnApplication(EitherVFX-preferthisone)
  (params (MyTarget :type Actor%20Object%20Reference) (Parameters :type Gameplay%20Cue%20Parameters%20Structure%20%28by%20ref%29) (ParticleComponents :type Array%20of%20FXSystem%20Component%20Object%20References) (AudioComponents :type Array%20of%20Audio%20Component%20Object%20References) (BurstCameraShakeInstance :type Camera%20Shake%20Base%20Object%20Reference) (BurstDecalInstance :type Decal%20Actor%20Object%20Reference))
  ;[0.0276 0.1823 0.05729] OnApplication: Cache the necessary variables and start drawing the indicator to the closest terminal.
  ;;[0.01103 0.0625 0.06063] Cache necessary variables.
  (bind _asfort_player_pawn_athena (Utilities|Casting|CastToFortPlayerPawnAthena MyTarget)
    (:then
      (bind _output_get (Variables|Default|SetTargetPlayerPawn _asfort_player_pawn_athena))
      (bind _asfort_player_controller_athena (Utilities|Casting|CastToFortPlayerControllerAthena (Pawn|GetController _output_get)))
      (Variables|Default|SetTargetPlayerController _asfort_player_controller_athena)
      (Fortnite|FortShippingLog self (Utilities|String|Append "OnApplication: PlayerController: " (Utilities|GetDisplayName _asfort_player_controller_athena)))
      (bind _output_get_2 (Variables|Default|SetGameplayCueParams Parameters))
      (Variables|Default|SetClosestLockedTerminal 0)
      ;;[0.138 0.1406 0.01869] Start the timer based on the desired frequency and draw the indicators.
      (if (> (.RawMagnitude (Ability|GameplayCue|BreakGameplayCueParameters _output_get_2)) 0)
        (Fortnite|FortShippingLog self (Utilities|String|Append "OnApplication: Invoked for player " (Fortnite|GetPlayerNameSafe (Variables|Default|GetTargetPlayerPawn)) " with an update frequency of " (Utilities|String|ToString(Float) (.RawMagnitude (Ability|GameplayCue|BreakGameplayCueParameters _output_get_2)))))
        (bind _returnvalue (Utilities|Time|SetTimerbyEvent (EventDispatchers|CreateEvent :!event Custom|DrawIndicatorToClosestTerminal) (.RawMagnitude (Ability|GameplayCue|BreakGameplayCueParameters _output_get_2)) true))
        (Variables|Default|SetDrawIndicatorTimerHandle _returnvalue)
        (CallFunction|DrawIndicatortoClosestTerminal)))
    (:CastFailed
      ;;;[1 0 0] Warning: Target is NOT FortPlayerPawnAthena
      (Fortnite|FortShippingLog self (Utilities|String|Append "OnApplication: Target is NOT FortPlayerPawnAthena: " (Utilities|GetDisplayName MyTarget)) true))))

(event Custom|DrawIndicatorToClosestTerminal
  ;[0.08421 0.01953 0.1354] DrawIndicatorToClosestTerminal: Find the closest terminal and draw the desired indicators.
  (bind _terminal (CallFunction|FindClosestKeycardAcceptingTerminal))
  (Utilities|IsValid _terminal
    (:Is%20Valid
      (Utilities|FlowControl|Sequence
        (:then_0
          (if (!= _terminal (Variables|Default|GetClosestLockedTerminal))
            ;;[0.03346 0.155 0.1615] Unmark any previously indicated terminal.
            (CallFunction|UnmarkClosestTerminal)
            ;;[0.01244 0.05729 0.03845] Mark the closest locked terminal and mark the actor on the client
            (Variables|Default|SetClosestLockedTerminal _terminal)
            (CallFunction|MarkClosestTerminal)))
        (:then_1
          (CallFunction|SpawnDirectionIndicatorFX) ; Always attempt to spawn the direction indicator FX every desired seconds.
        )))
    (:Is%20Not%20Valid
      ;;[1 0 0] Warning: Closest Locked Terminal is null.
      (Fortnite|FortShippingLog self "DrawIndicatorToClosestTerminal: Closest Locked Terminal is null." true))))

(event EventOnRemoval(EitherVFX-preferthisversion)
  (params (MyTarget :type Actor%20Object%20Reference) (Parameters :type Gameplay%20Cue%20Parameters%20Structure%20%28by%20ref%29) (ParticleComponents :type Array%20of%20FXSystem%20Component%20Object%20References) (AudioComponents :type Array%20of%20Audio%20Component%20Object%20References) (BurstCameraShakeInstance :type Camera%20Shake%20Base%20Object%20Reference) (BurstDecalInstance :type Decal%20Actor%20Object%20Reference))
  ;[0.07812 0.04087 0.01689] OnRemoval: Stop the running timer and unbind bound delegates.
  (Utilities|Time|ClearandInvalidateTimerbyHandle MyTarget (Variables|Default|GetDrawIndicatorTimerHandle)))

(fn FindClosestKeycardAcceptingTerminal
  ;[0.07812 0.07462 0.01764] Override in child.
  (return 0))

(fn MarkClosestTerminal
  ;[0.1069 0.375 0.166] Mark the closest terminal
  (bind _targetplayercontroller (Variables|Default|GetTargetPlayerController)
    (:then
      (bind (_playerbankid _bankinstanceid) (Fortnite|RegisterSoundIndicatorPlacementContext _targetplayercontroller (Variables|Default|GetClosestLockedTerminal) "0, 0, 0" false))
      (Variables|SetMarkerID_PlayerID (Utilities|Struct|MakeMarkerID _playerbankid _bankinstanceid)))
    (:else)))

(fn SpawnDirectionIndicatorFX
  (bind (_normalizedmagnitude _rawmagnitude _effectcontext _matchedtagname _originaltag _aggregatedsourcetags _aggregatedtargettags _location _normal _instigator _effectcauser _sourceobject _physicalmaterial _gameplayeffectlevel _abilitylevel) (Ability|GameplayCue|BreakGameplayCueParameters (Variables|Default|GetGameplayCueParams)))
  ;[0.03131 0.05622 0.1823] SpawnDirectionIndicatorFX: Spawn the direction FX with the correct orientation.
  (if (>= _abilitylevel 1)
    (bind _targetplayerpawn (Variables|Default|GetTargetPlayerPawn)
      (:then
        (bind _mesh (Variables|/Script/Engine.Character|GetMesh _targetplayerpawn)
          (:then
            (bind _returnvalue (Effects|SpawnEmitterAttached "/Game/Athena/Items/EnvironmentalItems/Locks/Keycard/FX/P_Athena_KeycardLock_Direction.P_Athena_KeycardLock_Direction" _mesh "root" "0, 0,125.000000" (Math|Rotator|MakeRotfromX (* (- (Transformation|GetActorLocation _targetplayerpawn) (Transformation|GetActorLocation (Variables|Default|GetClosestLockedTerminal))) -1.0)) "1.000000,1.000000,1.000000" "SnapToTarget"))
            (Variables|/Script/Engine.SceneComponent|SetbAbsoluteRotation true _returnvalue))
          (:else
            ;;[1 0 0] Warning: TargetPlayerPawn's mesh is null.
            (Fortnite|FortShippingLog self (Utilities|String|Append "SpawnDirectionIndicatorFX: TargetPlayerPawn Mesh is null." (Fortnite|GetPlayerNameSafe (Variables|Default|GetTargetPlayerPawn))) true))))
      (:else
        ;;[1 0 0] Warning: TargetPlayerPawn is null.
        (Fortnite|FortShippingLog self "SpawnDirectionIndicatorFX: TargetPlayerPawn is null." true)))
    (else
      (Utilities|Time|ClearandInvalidateTimerbyHandle (Variables|Default|GetTargetPlayerPawn) (Variables|Default|GetDrawIndicatorTimerHandle))
      (Fortnite|FortShippingLog self "SpawnDirectionIndicatorFX: Stopping cause AbilityLevel < 1."))))

(fn UnmarkClosestTerminal
  ;[0.0567 0.2292 0.1095] Unmark Closest Terminal: Unmark the closest terminal
  (bind _targetplayercontroller (Variables|Default|GetTargetPlayerController)
    (:then
      (Utilities|IsValid (Class|FortPlayerController|GetMarkerComponent _targetplayercontroller)
        (:Is%20Valid
          (MarkerComponent|UnmarkActoronClient (Class|FortPlayerController|GetMarkerComponent _targetplayercontroller) (Variables|Default|GetMarkerID)))
        (:Is%20Not%20Valid
          ;;[1 0 0] Warning: MarkerComponent is null.
          (Fortnite|FortShippingLog self (Utilities|String|Append "UnmarkClosestTerminal: MarkerComponent is null for player controller: " (Utilities|GetDisplayName _targetplayercontroller)) true))))
    (:else)))

(fn ConstructionScript)
